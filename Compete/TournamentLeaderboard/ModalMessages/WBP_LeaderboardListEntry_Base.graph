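# Base tile button widget for a tournament leaderboard list entry, derived from WBP_CompeteTileButton_Base.
# On Construct it registers its background and rounded outline as UIKit modular button blocks.
# A Content named slot takes the entry's contents.
# UpdateBackgroundColor uses index parity to set Color_Idle on the background's dynamic material, striping even rows faint white and odd rows transparent by default.
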WidgetBlueprint /Compete/TournamentLeaderboard/ModalMessages/WBP_LeaderboardListEntry_Base
graphs [
 EventGraph {nodes [
   Event Construct()
   > Parent: Construct()
   > InitModularBlocks()
  ]}
 InitModularBlocks {nodes [
   InitModularBlocks() {is=title K2Node_FunctionEntry}
   > RegisterModularBlock!(
    execute
    self:UIKit Modular Button Object Reference
    Block:UIKit Block Object Reference=Get Block_Background() -> (Block_Background:WBP UIKit Block Background Generic Object Reference)
   ) -> (then)
   > RegisterModularBlock(Block=Get Block_Outline() -> (Block_Outline:WBP UIKit Block Outline Object Reference))
  ]}
 UpdateBackgroundColor {nodes [
   n1 = UpdateBackgroundColor!() -> (then, Index:int) {is=title K2Node_FunctionEntry}
   n2 = "% (Integer)"(A=@n1, B=2) -> (ReturnValue=0)
   > SetVectorParameterValue(
    self=Get BackgroundDynamicMaterial!(self:WBP UIKit Block Background Base Object Reference=Get Block_Background()) -> (BackgroundDynamicMaterial:Material Instance Dynamic Object Reference)
    ParameterName=Color_Idle
    Value=Select(
     Option 0:Linear Color Structure=Get Color_1() -> (Color_1:Linear Color Structure)
     Option 1:Linear Color Structure=Get Color_2() -> (Color_2:Linear Color Structure)
     Index="Equal (Integer)"(A=@n2)
    ) -> (ReturnValue:Linear Color Structure)
   )
  ]}
]
variables [
 Color_1 {type=LinearColor, default=<rgba 0 0 0 0>}
 Color_2 {type=LinearColor, default=<rgba 1 1 1 0.051>}
]
widgets {
 info {
  inheritedWidgetCount=1
  parentClass=/Compete/WBP_CompeteTileButton_Base.WBP_CompeteTileButton_Base_C
  rootWidgetClass=&GridPanel
  widgetCount=5
 }
 widgets [
  Root {widgetClassPath=&GridPanel, widget=Root, properties {columnFill [1], rowFill [1]}}
  Block_Background {
   widgetClassPath=/FortUILibrary/Systems/ModularButton/Blocks/WBP_UIKit_Block_BackgroundGeneric
   bIsVariable=true
   parent=Root
   widget=Block_Background
   properties {backgroundMaterial=/Compete/MaterialInstances/MI_UI_Compete_TournamnetTile_LeaderboardItem_Background}
   slot {class=GridSlot}
  }
  Block_Outline {
   widgetClassPath=/FortUILibrary/UIKit/Buttons/Blocks/WBP_UIKit_Block_Outline
   bIsVariable=true
   parent=Root
   widget=Block_Outline
   properties {bottomLeftCornerRadius=25, bottomRightCornerRadius=25, radiusInPixels=true, topLeftCornerRadius=25, topRightCornerRadius=25}
   slot {class=GridSlot}
  }
  Content {widgetClassPath=&NamedSlot, bIsVariable=true, parent=Root, widget=Content, slot {class=GridSlot}}
  InputActionWidget {widgetClassPath=&CommonActionWidget, bInherited=true, widget=null}
 ]
}
